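; Character part modifier for the Nitro Gauntlet that shows the player's current weapon ammo, clamped to 0–4, on the part's skeletal mesh.
; After res-in, it writes remaining ammo to custom primitive data slot 1 and binds to local ammo changes.
; Ammo gains stamp game time into slot 2, then clear it and update the count after a half-second delay.
; Other changes update immediately; every change prints a debug string.
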(event EventOnPostResIn
  (bind _returnvalue (CustomCharacter|GetAssociatedPlayerPawn))
  (bind _returnvalue_2 (Weapon|GetCurrentWeapon _returnvalue))
  (|CallParentFunction)
  (Rendering|Material|SetDefaultCustomPrimitiveDataFloat (CustomCharacter|GetSkeletalMeshForPartOnPlayerPawn) 1 (Math|Conversions|ToFloat(Integer) (Math|Integer|Clamp(Integer) (FortWeapon|GetRemainingAmmo _returnvalue_2) 0 4)))
  (EventDispatchers|BindEventtoOnLocalAmmoChanged _returnvalue_2 (delegate Custom|UpdatingAmmoCount)))

(event Custom|UpdatingAmmoCount
  (params (LocalCount :type Integer) (LocalRemaining :type Integer))
  (bind _returnvalue (CustomCharacter|GetSkeletalMeshForPartOnPlayerPawn))
  (Variables|Default|SetPreviousAmmoCount (Variables|Default|GetCurrentAmmoCount))
  (Variables|Default|SetCurrentAmmoCount (Math|Integer|Clamp(Integer) LocalRemaining 0 4))
  (Development|PrintString (Utilities|String|Append "AMMO REMAINING: " (Utilities|String|ToString(Integer) LocalRemaining)))
  (if (> (Variables|Default|GetCurrentAmmoCount) (Variables|Default|GetPreviousAmmoCount))
    (Rendering|Material|SetCustomPrimitiveDataFloat _returnvalue 2 (Utilities|Time|GetGameTimeinSeconds))
    (Utilities|FlowControl|Delay 0.5)
    (Rendering|Material|SetCustomPrimitiveDataFloat _returnvalue 2 0.0)
    (Rendering|Material|SetCustomPrimitiveDataFloat _returnvalue 1 (Math|Conversions|ToFloat(Integer) (Variables|Default|GetCurrentAmmoCount)))
    (else
      (Rendering|Material|SetCustomPrimitiveDataFloat _returnvalue 1 (Math|Conversions|ToFloat(Integer) (Variables|Default|GetCurrentAmmoCount)))
      (Rendering|Material|SetCustomPrimitiveDataFloat _returnvalue 2 0.0))))

(fn ConstructionScript)
